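# Material function that turns masked areas of the Cat Burglar Ghost cosmetic into animated, glowing ghostly smoke.
# Time-scrolled triplanar noise in pre-skinned space warps a second mask, whose channels blend smoke colors within height gradients.
# Vertex-color blue gates the smoke, normal noise and world-space-noise vertex displacement, blended over input attributes by Mask Map red.
# Fresnel edge glow, a sine-driven glow in mask alpha and a two-color emissive map layer on top.
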MaterialFunction /BRCosmetics/Characters/Player/Male/Medium/Bodies/M_MED_Cat_Burglar_Ghost/MF_Cat_Burglar_Ghost
graph {nodes [
  n1 = FunctionInput() {bUsePreviewValueAsDefault=true, inputName=Material, inputType=FunctionInput_MaterialAttributes}
  n2 = VectorParameter() {group=Ghost Arsenic, parameterName=Ghost_Base Color, defaultValue=<rgba 0.001 0.001 0.001 0.001>}
  n3 = ScalarParameter() {defaultValue=0.781, group=Ghost Arsenic, parameterName=Ghost_Roughness}
  n4 = VectorParameter() {group=Ghost Arsenic, parameterName=Ghost_Smoke_BaseColor, defaultValue=<rgba 0.0363 0.0045 0.0521 1>}
  n5 = VectorParameter() {group=Ghost Arsenic, parameterName=Ghost_Smoke Streams Tile Rate, defaultValue=<rgba 2 1 1 1>}
  n6 = PreSkinnedLocalBounds()
  n7 = Subtract(A=VertexInterpolator(VS=PreSkinnedPosition()).PS, B=@n6.Min)
  n8 = Subtract(A=@n6.Max, B=@n6.Max)
  n9 = MaterialFunctionCall(Color=@n8) {materialFunction=&SplitComponents}
  n10 = Divide(A=@n7, B=Max(A=Max(A=@n9.R, B=@n9.R), B=@n9.B))
  n11 = Constant() {r=1}
  n12 = Time()
  n13 = VectorParameter() {group=Ghost Arsenic, parameterName=Ghost_Smoke_Warp_Speed, defaultValue=<rgba 0.01 0.03 -0.1 0>}
  n14 = Multiply(
   A=Add(A=@n10, B=Multiply(A=@n12, B=@n13.RGB))
   B=ScalarParameter() {defaultValue=1, group=Ghost Arsenic, parameterName=Ghost_Smoke_Warp_Tile_Rate}
  )
  n15 = ScalarParameter() {group=Ghost Arsenic, parameterName=Ghost ProjectionTransitionContrast}
  n16 = TextureObject() {texture=/BRCosmetics/Characters/Player/Male/Medium/Bodies/M_MED_Cat_Burglar_Ghost/Textures/T_M_MED_Cat_Burglar_Ghost_Mask4}
  n17 = VertexInterpolator(VS=PreSkinnedNormal())
  n18 = Multiply(
   A=ScalarParameter() {defaultValue=-0.5, group=Ghost Arsenic, parameterName=Ghost_Smoke_Warp_Amount}
   B=ComponentMask(None=MaterialFunctionCall(TextureObject=@n16, TextureSize=@n11, WorldPosition=@n14, World Space Normal=@n17.PS, ProjectionTransitionContrast=@n15) {materialFunction=&WorldAlignedTexture}.XYZ Texture) {r=true}
  )
  n19 = Divide(A=@n7, B=@n8)
  n20 = Divide(
   A=Add(A=@n19, B=ScalarParameter() {defaultValue=-1.25, group=Ghost Arsenic, parameterName=Ghost_Z_Warp_Masking_Gradent_Offset})
   B=ScalarParameter() {defaultValue=-0.5, group=Ghost Arsenic, parameterName=Ghost_Z_Warp_Masking_Gradent_Falloff}
  )
  n21 = Saturate(None=ComponentMask(None=Power(Base=@n20)) {b=true})
  n22 = MaterialFunctionCall(
   TextureObject=TextureObject() {texture=/BRCosmetics/Characters/Player/Male/Medium/Bodies/M_MED_Cat_Burglar_Ghost/Textures/T_M_MED_Cat_Burglar_Ghost_Mask3}
   TextureSize=Constant3Vector() {constant=<rgba 1 1 1 0>}
   WorldPosition=Add(A=Multiply(A=@n10, B=@n5.RGB), B=Multiply(A=@n18, B=@n21))
   Export Float 4=StaticBool() {value=true}
   ProjectionTransitionContrast=@n15
  ) {materialFunction=&WorldAlignedTexture}
  n23 = ComponentMask(None=@n22.XYZ Texture) {g=true}
  n24 = ScalarParameter() {defaultValue=-1.15, group=Ghost Arsenic, parameterName=Ghost_Z_Smoke_Masking_Gradent_OffsetGhost_}
  n25 = Power(Base=Divide(A=Add(A=@n19, B=@n24), B=ScalarParameter() {defaultValue=-0.8, group=Ghost Arsenic, parameterName=Ghost_Z_Smoke_Masking_Gradent_Falloff}))
  n26 = ScalarParameter() {defaultValue=0.5, group=Ghost Arsenic, parameterName=Ghost_Z_Smoke_Masking_Head_Gradent_Offset}
  n27 = Power(Base=Divide(A=Add(A=@n19, B=@n26), B=ScalarParameter() {defaultValue=0.1, group=Ghost Arsenic, parameterName=Ghost_Z_Smoke_Masking_Head_Gradent_Falloff}))
  n28 = Multiply(A=Saturate(None=ComponentMask(None=@n25) {b=true}), B=Saturate(None=ComponentMask(None=@n27) {b=true}))
  n29 = LinearInterpolate(
   A=LinearInterpolate(
    A=@n4.RGB
    B=VectorParameter() {group=Ghost Arsenic, parameterName=Ghost_Smoke_Color_C, defaultValue=<rgba 0 0.4211 1 1>}.RGB
    Alpha=Multiply(A=@n23, B=@n28)
   )
   B=VectorParameter() {group=Ghost Arsenic, parameterName=Ghost_Smoke_Color_D, defaultValue=<rgba 0.0316 0.105 1 1>}.RGB
   Alpha=Multiply(A=ComponentMask(None=@n22.XYZ Texture) {b=true}, B=@n28)
  )
  n30 = VectorParameter() {group=Ghost Arsenic, parameterName=Ghost_Smoke_Color_A, defaultValue=<rgba 0.5592 0.0683 1 1>}
  n31 = Multiply(A=ComponentMask(None=@n22.XYZ Texture) {r=true}, B=@n28)
  n32 = ComponentMask(None=@n22.XYZ Texture) {a=true}
  n33 = Saturate(None=Add(A=VertexColor().B, B=ScalarParameter() {group=Ghost Arsenic, parameterName=Ghost_Disable_Vertex_Colors}))
  n34 = LinearInterpolate(
   A=@n4.RGB
   B=LinearInterpolate(
    A=LinearInterpolate(A=@n29, B=@n30.RGB, Alpha=@n31)
    B=VectorParameter() {group=Ghost Arsenic, parameterName=Ghost_Smoke_Color_B, defaultValue=<rgba 0.5592 0.0683 1 1>}.RGB
    Alpha=Multiply(A=@n32, B=@n28)
   )
   Alpha=@n33
  )
  n35 = Constant3Vector() {constant=<rgba 0 0 1 0>}
  n36 = TextureObject() {
   samplerType=SAMPLERTYPE_Normal
   texture=/Game/Characters/Player/Male/Medium/Bodies/M_MED_Cat_Burglar_Ghost/Textures/T_M_MED_Cat_Burglar_Ghost_N1
  }
  n37 = Multiply(A=Time(), B=VectorParameter() {group=Ghost Arsenic, parameterName=Ghost Normal Noise Speed, defaultValue=<rgba 0.01 0.03 -0.05 0>}.RGB)
  n38 = MaterialFunctionCall!(
   TextureObject=@n36
   TextureSize=Constant() {r=1}
   RotYAxis
   RotZAxis
   RotXAxis
   PreSkinNormal=@n17.PS
   ProjectionTransitionContrast=@n15
   WorldPosition=Multiply(A=Add(A=@n37, B=@n10), B=ScalarParameter() {defaultValue=10, group=Ghost Arsenic, parameterName=Ghost Normal Noise Tile_RAte})
  ) {materialFunction=/Game/Packages/Fortress_SharedMaterials/Base_Material_Functions/BaseCustomizationFunctions/PreSkinnedAlignedNormal}
  n39 = GetMaterialAttributes(None=@n1) {attributeGetTypes [#1]}
  n40 = LinearInterpolate(
   A=@n35
   B=LinearInterpolate(
    A=@n39.Normal
    B=Add(A=@n39.Normal, B=@n38.XYZ Texture)
    Alpha=ScalarParameter() {defaultValue=1, group=Ghost Arsenic, parameterName=Ghost Normal Noise_Stregth}
   )
   Alpha=@n33
  )
  n41 = ShadingPathSwitch(Default=@n40, Deferred=@n40, Mobile=@n35) {desc=Migrated from FeatureLevelSwitch}
  n42 = VertexColor()
  n43 = ScalarParameter() {group=Ghost Arsenic, parameterName=Ghost_Disable_Vertex_Colors}
  n44 = Constant() {r=1}
  n45 = Add(
   A=Multiply(A=WorldPosition().XYZ, B=ScalarParameter() {defaultValue=0.01, group=Ghost Arsenic, parameterName=Ghost_WPO_Tile_Rate})
   B=Multiply(A=Time(), B=VectorParameter() {group=Ghost Arsenic, parameterName=Ghost_WPO_Speed, defaultValue=<rgba 0.02 0.1 -0.2 1>}.RGB)
  )
  n46 = VertexNormalWS()
  n47 = Add(A=ComponentMask(None=MaterialFunctionCall(TextureObject=@n16, TextureSize=@n44, WorldPosition=@n45, World Space Normal=@n46) {materialFunction=&WorldAlignedTexture}.XYZ Texture) {r=true}) {constB=-0.5}
  n48 = Multiply(
   A=Saturate(None=Add(A=@n42.B, B=@n43))
   B=Multiply(A=Multiply(A=@n47, B=@n46), B=ScalarParameter() {defaultValue=3, group=Ghost Arsenic, parameterName=Ghost_WPO_Amount})
  )
  n49 = TextureObjectParameter() {
   group=Base Textures
   parameterName=Mask Map
   sortPriority=0
   texture=/BRCosmetics/Characters/Player/Male/Medium/Bodies/M_MED_Cat_Burglar_Ghost/Textures/T_M_MED_Cat_Burglar_Ghost_Body_Mask
  }
  n50 = TextureSample(Tex=@n49) {texture=&T_BlacksRGB}
  n51 = GetMaterialAttributes(None=BlendMaterialAttributes(
    A=@n1
    B=MakeMaterialAttributes!(
     BaseColor=@n2.RGB
     Metallic
     Specular
     Roughness=@n3
     Anisotropy
     EmissiveColor=@n34
     Opacity
     OpacityMask
     Normal=@n41
     Tangent
     WorldPositionOffset=@n48
     SubsurfaceColor
     ClearCoat
     ClearCoatRoughness
     AmbientOcclusion
     Refraction
     CustomizedUVs_0
     CustomizedUVs_1
     CustomizedUVs_2
     CustomizedUVs_3
     CustomizedUVs_4
     CustomizedUVs_5
     CustomizedUVs_6
     CustomizedUVs_7
     PixelDepthOffset
     ShadingModel
     Displacement
    )
    Alpha=@n50.R
   )) {attributeGetTypes [#2, #1]}
  n52 = Fresnel(
   ExponentIn=ScalarParameter() {defaultValue=2.5, group=Ghost EgdeGlow, parameterName=Fresnel Power, sortPriority=2}
   BaseReflectFractionIn=Constant()
   Normal=Transform(None=@n51.Normal)
  ) {baseReflectFraction=0}
  n53 = VectorParameter() {group=Ghost EgdeGlow, parameterName=Fresnel Color, sortPriority=0, defaultValue=<rgba 0.1258 0.0113 0.14 0>}
  n54 = Add(A=Multiply(A=Saturate(None=@n52), B=Add(A=@n53.RGB, B=@n34)), B=@n51.EmissiveColor)
  n55 = GetMaterialAttributes(None=BlendMaterialAttributes(
    A=@n51.MaterialAttributes
    B=SetMaterialAttributes!(MaterialAttributes=@n51.MaterialAttributes, EmissiveColor=@n54) {attributeSetTypes [#2]}
    Alpha=ScalarParameter() {defaultValue=1, group=Ghost EgdeGlow, parameterName=Fresnel Blend Amount, sortPriority=-2}
   )) {attributeGetTypes [#2, #3]}
  n56 = VectorParameter() {group=Ghost EgdeGlow, parameterName=Glow Base Color, sortPriority=-10, defaultValue=<rgba 0.1258 0.0113 0.14 0>}
  n57 = Sine(None=Multiply(A=Time(), B=ScalarParameter() {defaultValue=0.25, group=Ghost EgdeGlow, parameterName=Base Time, sortPriority=37}))
  n58 = Clamp(
   None=Divide(A=Add(A=@n57))
   Min=ScalarParameter() {group=Ghost EgdeGlow, parameterName=Whisp Min Brightness, sortPriority=35}
   Max=ScalarParameter() {defaultValue=1, group=Ghost EgdeGlow, parameterName=Whisp Max Brightness, sortPriority=36}
  ) {maxDefault=0.75, minDefault=0.25}
  n59 = LinearInterpolate(A=@n55.EmissiveColor, B=Add(A=Multiply(A=@n55.EmissiveColor) {constB=0.5}, B=Multiply(A=@n56.RGB, B=@n58)), Alpha=@n50.A)
  n60 = LinearInterpolate(A=@n55.BaseColor, B=@n2.RGB, Alpha=ScalarParameter() {group=Ghost EgdeGlow, parameterName=Use Base Color Override for Glow, sortPriority=50})
  n61 = GetMaterialAttributes(None=BlendMaterialAttributes(
    A=@n55.MaterialAttributes
    B=SetMaterialAttributes!(MaterialAttributes=@n55.MaterialAttributes, EmissiveColor=@n59, BaseColor=@n60) {attributeSetTypes [#2, #3]}
    Alpha=@n50.A
   )) {attributeGetTypes [#2]}
  n62 = TextureSample(Tex=TextureObjectParameter() {group=Base Textures, parameterName=Emissive Map, sortPriority=0, texture=&T_BlacksRGB}) {texture=&T_BlacksRGB}
  n63 = LinearInterpolate(
   A=VectorParameter() {group=Ghost EgdeGlow, parameterName=Emissive Texture Color A, sortPriority=-10}.RGB
   B=VectorParameter() {group=Ghost EgdeGlow, parameterName=Emissive Texture Color B, sortPriority=-10}.RGB
   Alpha=@n62.R
  )
  FunctionOutput(None=BlendMaterialAttributes(
    A=@n61.MaterialAttributes
    B=SetMaterialAttributes!(MaterialAttributes=@n61.MaterialAttributes, EmissiveColor=@n63) {attributeSetTypes [#2]}
    Alpha=@n62.R
   ))
 ]}
