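# Master material for the Cat Burglar Ghost character body, built on the shared character PreFX and PostFX functions.
# MF_Cat_Burglar_Ghost restyles the PreFX output across the whole body by default.
# The Use_SL_Ghost switch instead confines the ghost's color, roughness, emissive and normal to a mask combining the SuperLevelling mask with Mask Map alpha.
# That branch also adds the ghost-blended emissive onto the SuperLevelling emissive.
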Material /BRCosmetics/Characters/Player/Male/Medium/Bodies/M_MED_Cat_Burglar_Ghost/M_FN_Character_Cat_Burglar_Ghost_Master
graph {
 properties {
  bUseMaterialAttributes=true
  bUsedWithMeshDeformer=true
  bUsedWithMorphTargets=true
  bUsedWithSkeletalMesh=true
  blendMode=BLEND_Masked
  pixelDepthOffsetMode=PDOM_Legacy
  usedShadingModels=Default Lit
 }
 nodes [
  n1 = MaterialFunctionCall() {materialFunction=&BaseFN_CharacterFunction_PreFX}
  n2 = MaterialFunctionCall!(Local Normal, Local Position, Material=@n1.Result) {materialFunction=/Game/Packages/Fortress_SharedMaterials/Base_Material_Functions/MF_SuperLevellingMask}
  n3 = GetMaterialAttributes(None=@n2.Result) {attributeGetTypes [#1]}
  n4 = GetMaterialAttributes(None=@n1.Result) {attributeGetTypes [
    #2
    #3
    #1
    #4
   ]}
  n5 = MaterialFunctionCall!(Material=@n1.Result) {materialFunction=&/MF_Cat_Burglar_Ghost}
  n6 = GetMaterialAttributes(None=@n5.Result) {attributeGetTypes [
    #2
    #3
    #1
    #4
   ]}
  n7 = TextureSampleParameter2D() {
   group=Masks
   parameterName=Mask Map
   samplerSource=SSM_Wrap_WorldGroupSettings
   samplerType=SAMPLERTYPE_Masks
   texture=&T_BlackMask
  }
  n8 = Saturate(None=Add(A=ComponentMask(None=@n2.SuperLevelling_Mask) {r=true}, B=@n7.A))
  n9 = SetMaterialAttributes!(
   MaterialAttributes=@n3.MaterialAttributes
   Base Color=LinearInterpolate(A=@n4.Base Color, B=@n6.Base Color, Alpha=@n8)
   Roughness=LinearInterpolate(A=@n4.Roughness, B=@n6.Roughness, Alpha=@n8)
   Emissive Color=Add(A=@n3.Emissive Color, B=LinearInterpolate(A=@n4.Emissive Color, B=@n6.Emissive Color, Alpha=@n8))
   Normal=LinearInterpolate(A=@n4.Normal, B=@n6.Normal, Alpha=@n8)
  ) {attributeSetTypes [
    #2
    #3
    #1
    #4
   ]}
  MaterialFunctionCall(
   AO=@n1.AO
   Distance Blend=@n1.AO
   Fresnel=@n1.AO
   Gmap=@n1.AO
   Pre Skinned Local Position=@n1.AO
   Material=StaticSwitchParameter(True=@n9, False=@n5.Result) {parameterName=Use_SL_Ghost}
  ) {materialFunction=&BaseFN_CharacterFunction_PostFX}
 ]
}
